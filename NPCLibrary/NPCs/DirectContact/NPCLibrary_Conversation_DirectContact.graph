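# Conversation database for the DirectContact NPC, who introduces herself as Katara and speaks when a player talks to her.
# The entry point feeds one Speech node with a default line and contextual messages gated by first-ever conversation, specific player outfits or match conversation count.
# Curve-row priorities rank the introduction and outfit reactions.
# The Speech node then branches through link nodes into the shared NPC Library Quests and BasicServices conversations.
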ConversationDatabase /NPCLibrary/NPCs/DirectContact/NPCLibrary_Conversation_DirectContact
entry_points [{tag=Conversation.Entry.DirectContact, destinations [#1]}]
nodes [
 ConversationEntryPointNode_7 {
  id=#1
  class=ConversationEntryPointNode
  outputs [#2]
  properties {entryTag {tagName=Conversation.Entry.DirectContact}}
 }
 Speech_C_4 {
  id=#2
  class=Speech_C
  outputs [#3, #4]
  properties {
   nodeName=""
   generalConfig {
    defaultMessage=Don't let anyone take away your courage.
    contextualMessages [
     {
      message="I'm Katara. You haven't seen my brother around, have you?"
      contextRequirements [{requirement {class=FortControllerRequirement_IsFirstEverConversationWithNPC}}]
      priority {curve {rowName=Tandem.Speech.Priority.FirstEverInteraction}}
     }
     {
      message=I should have known you'd follow Aang. Even here.
      contextRequirements [{requirement {class=FortControllerRequirement_HasCID, properties {allowedCIds [/BRCosmetics/Athena/Items/Cosmetics/Characters/Character_TikiTorch]}}}]
      priority {curve {rowName=Tandem.Speech.Priority.OutfitReaction}}
     }
     {
      message="Hey, how come you never call me Sifu Katara, anyway?"
      contextRequirements [{requirement {class=FortControllerRequirement_HasCID, properties {allowedCIds [/BRCosmetics/Athena/Items/Cosmetics/Characters/Character_DoubleDuty]}}}]
      priority {curve {rowName=Tandem.Speech.Priority.OutfitReaction}}
     }
     {
      message="Maybe you should wash up? You've got a little dirt on your...everywhere, actually."
      contextRequirements [{requirement {class=FortControllerRequirement_HasCID, properties {allowedCIds [/BRCosmetics/Athena/Items/Cosmetics/Characters/Character_RemoteControl]}}}]
      priority {curve {rowName=Tandem.Speech.Priority.OutfitReaction}}
     }
     {
      message=Don't let anyone take away your courage.
      contextRequirements [{requirement {
         class=FortControllerRequirement_CheckNPCMatchConversationCount
         properties {nPCUniqueIdTag {tagName=AISpawnerData.Type.Tandem.DirectContact}}
        }}]
      priority {curve {rowName=None}}
     }
     {
      message="Keep fighting, even if you don't think you'll win."
      contextRequirements [{requirement {
         class=FortControllerRequirement_CheckNPCMatchConversationCount
         properties {nPCUniqueIdTag {tagName=AISpawnerData.Type.Tandem.DirectContact}}
        }}]
      priority {curve {rowName=None}}
     }
     {
      message=Waterbending is hard. Don't be frustrated if you don't get it right away.
      contextRequirements [{requirement {
         class=FortControllerRequirement_CheckNPCMatchConversationCount
         properties {nPCUniqueIdTag {tagName=AISpawnerData.Type.Tandem.DirectContact}}
        }}]
      priority {curve {rowName=None}}
     }
     {
      message=Stay safe out there. Don't give up hope.
      contextRequirements [{requirement {
         class=FortControllerRequirement_CheckNPCMatchConversationCount
         properties {nPCUniqueIdTag {tagName=AISpawnerData.Type.Tandem.DirectContact}}
        }}]
      priority {curve {rowName=None}}
     }
    ]
   }
  }
 }
 ConversationLinkNode_0 {id=#4, class=ConversationLinkNode, properties {remoteEntryTag {tagName=Conversation.Entry.BasicServices}}}
 ConversationLinkNode_1 {id=#3, class=ConversationLinkNode, properties {remoteEntryTag {tagName=Conversation.Entry.NPCLibrary.Quests}}}
]
